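# Sound cue for the DorsalDance cosmetic pickaxe swing, mixing a close layer with a medium-distance layer, each under its own attenuation.
# The close layer switches on the SwingIdx parameter: three slots share random standard swings, while the last plays a slightly delayed, boosted random long swing.
# The medium layer plays random generic fire axe distant swings.
# A concurrency override allows one instance, stopping the oldest.
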SoundCue /BRCosmetics/Sounds/PickAxes/DorsalDance/PickaxeSwing_DorsalDance_Cue
root=Mixer_1
nodes [
 {id=Attenuation_0, children [Modulator_3], properties {attenuationSettings=/Game/Sounds/Attenuation/Fort_Weapon_Pickaxe_Close_Att}}
 {id=Attenuation_4, children [Modulator_4], properties {attenuationSettings=/Game/Sounds/Attenuation/Fort_Weapon_Pickaxe_Medium_Att}}
 {id=Delay_0, children [Random_2], properties {delayMax=0.05, delayMin=0.05}}
 {id=Mixer_1, children [Attenuation_0, Attenuation_4], properties {inputVolume [1,1]}}
 {id=Modulator_0, children [Delay_0], properties {volumeMax=1.6, volumeMin=1.5}}
 {id=Modulator_3, children [Switch_1]}
 {id=Modulator_4, children [Random_4]}
 {
  id=Random_0
  children [WavePlayer_6, WavePlayer_7, WavePlayer_8, WavePlayer_9, WavePlayer_10, WavePlayer_11]
  properties {weights [1,1,1,1,1,1]}
 }
 {id=Random_2, children [WavePlayer_12, WavePlayer_13, WavePlayer_14, WavePlayer_0], properties {weights [1,1,1,1]}}
 {
  id=Random_4
  children [WavePlayer_1, WavePlayer_2, WavePlayer_3, WavePlayer_4, WavePlayer_5, WavePlayer_21]
  properties {weights [1,1,1,1,1,1]}
 }
 {id=Switch_1, children [null,Random_0,Random_0,Random_0,Modulator_0], properties {intParameterName=SwingIdx}}
 {id=WavePlayer_0, properties {soundWaveAssetPtr=&/Pickaxe_DorsalDance_Swing_Long_04}}
 {id=WavePlayer_1, properties {soundWaveAssetPtr=/Game/Athena/Sounds/Weapons/Fire_Axe_Swing_Distant_01}}
 {id=WavePlayer_10, properties {soundWaveAssetPtr=&/Pickaxe_DorsalDance_Swing_05}}
 {id=WavePlayer_11, properties {soundWaveAssetPtr=&/Pickaxe_DorsalDance_Swing_06}}
 {id=WavePlayer_12, properties {soundWaveAssetPtr=&/Pickaxe_DorsalDance_Swing_Long_01}}
 {id=WavePlayer_13, properties {soundWaveAssetPtr=&/Pickaxe_DorsalDance_Swing_Long_02}}
 {id=WavePlayer_14, properties {soundWaveAssetPtr=&/Pickaxe_DorsalDance_Swing_Long_03}}
 {id=WavePlayer_2, properties {soundWaveAssetPtr=/Game/Athena/Sounds/Weapons/Fire_Axe_Swing_Distant_02}}
 {id=WavePlayer_21, properties {soundWaveAssetPtr=/Game/Athena/Sounds/Weapons/Fire_Axe_Swing_Distant_06}}
 {id=WavePlayer_3, properties {soundWaveAssetPtr=/Game/Athena/Sounds/Weapons/Fire_Axe_Swing_Distant_03}}
 {id=WavePlayer_4, properties {soundWaveAssetPtr=/Game/Athena/Sounds/Weapons/Fire_Axe_Swing_Distant_04}}
 {id=WavePlayer_5, properties {soundWaveAssetPtr=/Game/Athena/Sounds/Weapons/Fire_Axe_Swing_Distant_05}}
 {id=WavePlayer_6, properties {soundWaveAssetPtr=&/Pickaxe_DorsalDance_Swing_01}}
 {id=WavePlayer_7, properties {soundWaveAssetPtr=&/Pickaxe_DorsalDance_Swing_02}}
 {id=WavePlayer_8, properties {soundWaveAssetPtr=&/Pickaxe_DorsalDance_Swing_03}}
 {id=WavePlayer_9, properties {soundWaveAssetPtr=&/Pickaxe_DorsalDance_Swing_04}}
]
properties {
 audioPropertiesSheet=/BRCosmetics/Sounds/Settings/PropertySheets/Pickaxe/APS_Cosmetic_Pickaxe_Swing
 bOverrideConcurrency=true
 maxDistance=3000
 soundClassObject=&Fort_WeaponMelee_PSM
 volumeMultiplier=0.16
 concurrencyOverrides {maxCount=1, resolutionRule=StopOldest}
 concurrencySet [&Player_Melee_Weapon_Conc]
}
